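# Masked, unlit Niagara sprite material for leaf particles in the Pumpkin vehicle boost cosmetic.
# It adds a leaf diffuse texture, scaled by dynamic Brightness, to a Fresnel glow driven by the normal map, squared GlowColor and dynamic FresnelBrightness.
# The sum passes through MF_FX_EmissiveAdjustments, while a separate chain multiplies particle alpha by the normal texture's alpha.
# The AnimationSpeed and TextureSwitch dynamic parameters are declared but unused.
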Material /VehicleCosmetics/Boosts/Boost_Pumpkin/Materials/M_Boost_Leaf
graph {
 properties {bUsedWithNiagaraSprites=true, blendMode=BLEND_Masked, pixelDepthOffsetMode=PDOM_Legacy, shadingModel=MSM_Unlit, usedShadingModels=Unlit}
 nodes [
  n1 = TextureCoordinate()
  n2 = TextureSampleParameter2D(UVs=@n1) {parameterName=Diffuse_01, texture=/VehicleCosmetics/Boosts/Boost_Pumpkin/Textures/T_VCVFX_Boost_Pumpkin_Leaf_02_FB_D}
  n3 = DynamicParameter() {paramNames [AnimationSpeed, Brightness, FresnelBrightness, TextureSwitch]}
  n4 = VectorParameter() {parameterName=GlowColor, defaultValue=<rgba 0.891 0.268 0.0385 0.5>}
  n5 = TextureSampleParameter2D(UVs=@n1) {parameterName=Normal_01, texture=/VehicleCosmetics/Boosts/Boost_Pumpkin/Textures/T_VCVFX_Boost_Pumpkin_Leaf_02_FB_N}
  n6 = Fresnel(Normal=@n5.RGB)
  n7 = Multiply(A=Multiply(A=@n4.RGB, B=@n4.RGB), B=Multiply(A=@n3.FresnelBrightness, B=Saturate(None=@n6)))
  MaterialFunctionCall(Emissive=Add(A=Multiply(A=@n2.RGB, B=@n3.Brightness), B=Saturate(None=@n7))) {materialFunction=&MF_FX_EmissiveAdjustments}
  Saturate(None=Multiply(A=ParticleColor().A, B=@n5.A))
 ]
}
