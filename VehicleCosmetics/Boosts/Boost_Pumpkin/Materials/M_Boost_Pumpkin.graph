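# Masked, unlit Niagara sprite material for the pumpkin vehicle boost effect.
# A dynamic parameter, TextureSwitch, is rounded to choose between two diffuse and normal texture sets.
# The normals drive a tangent-space Fresnel that adds an orange glow to brightness-scaled diffuse and alpha-weighted particle color, passed through MF_FX_EmissiveAdjustments.
# Particle alpha multiplies the normal textures' alpha, which serves as a mask.
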Material /VehicleCosmetics/Boosts/Boost_Pumpkin/Materials/M_Boost_Pumpkin
graph {
 properties {bUsedWithNiagaraSprites=true, blendMode=BLEND_Masked, pixelDepthOffsetMode=PDOM_Legacy, shadingModel=MSM_Unlit, usedShadingModels=Unlit}
 nodes [
  Constant() {r=2}
  n1 = TextureCoordinate()
  n2 = TextureSample(UVs=@n1) {samplerType=SAMPLERTYPE_LinearColor, texture=/VehicleCosmetics/Boosts/Boost_Pumpkin/Textures/T_VCVFX_Boost_Pumpkin_01_FB_N}
  n3 = TextureSample(UVs=@n1) {samplerType=SAMPLERTYPE_LinearColor, texture=/VehicleCosmetics/Boosts/Boost_Pumpkin/Textures/T_VCVFX_Boost_Pumpkin_00_FB_N}
  n4 = DynamicParameter() {defaultValue=<rgba 0.1 1 1 1>, paramNames [Pumpkin_Glow, Brightness, FresnelBrightness, TextureSwitch]}
  n5 = Ceil(None=@n4.TextureSwitch)
  n6 = If(A=@n4.TextureSwitch, B=Constant() {r=0.49}, "A > B"=@n5, "A == B"=@n5, "A < B"=Floor(None=@n4.TextureSwitch))
  n7 = LinearInterpolate(A=@n2.RGBA, B=@n3.RGBA, Alpha=@n6)
  Fresnel(Normal=@n7) {baseReflectFraction=-0.3, exponent=3}
  n8 = ParticleColor()
  n9 = TextureSample(UVs=@n1) {texture=/VehicleCosmetics/Boosts/Boost_Pumpkin/Textures/T_VCVFX_Boost_Pumpkin_01_FB_D}
  n10 = TextureSample(UVs=@n1) {texture=/VehicleCosmetics/Boosts/Boost_Pumpkin/Textures/T_VCVFX_Boost_Pumpkin_00_FB_D}
  n11 = LinearInterpolate(A=@n9.A, B=@n10.A, Alpha=@n6)
  n12 = Multiply(A=LinearInterpolate(A=@n9.RGBA, B=@n10.RGBA, Alpha=@n6), B=@n4.Brightness)
  n13 = Constant4Vector() {constant=<rgba 0.891 0.268 0.0385 8>}
  n14 = CameraVectorWS()
  n15 = Multiply(
   A=MaterialFunctionCall(
    Normal Vector=@n7
    Camera Vector=Transform(None=@n14) {transformSourceType=TRANSFORMSOURCE_World, transformType=TRANSFORM_Tangent}
    Power=Constant() {r=3}
   ) {materialFunction=&Fresnel_Function}.Result
   B=@n4.Pumpkin_Glow
  ) {constB=0.1}
  n16 = Multiply(A=Multiply(A=@n13.RGBA, B=@n13.RGBA), B=Multiply(A=@n4.FresnelBrightness, B=Saturate(None=@n15)))
  MaterialFunctionCall(Emissive=Add(A=Multiply(A=@n8.RGBA, B=@n11), B=Add(A=@n12, B=@n16))) {materialFunction=&MF_FX_EmissiveAdjustments}
  MaterialFunctionCall(X=Power(), Y=Power()) {materialFunction=&MakeFloat3}
  Reroute(None=@n7)
  Saturate(None=Multiply(A=@n8.A, B=LinearInterpolate(A=@n2.A, B=@n3.A, Alpha=@n6)))
 ]
}
