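# Unlit translucent Niagara sprite material for the flame of the Pumpkin boost cosmetic.
# A panning flow normal map distorts the UVs of a flame texture parameter and smoke mask, which are summed.
# Particle color and dynamic-parameter brightness drive emissive, while an unwarped smoke mask, particle alpha and depth fade shape opacity.
# Dynamic parameters also control distortion tiling and strength, plus a separate smoke-mask distortion term.
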Material /VehicleCosmetics/Boosts/Boost_Pumpkin/Materials/M_Boost_PumpkinFlame
graph {
 properties {bUsedWithNiagaraSprites=true, blendMode=BLEND_Translucent, pixelDepthOffsetMode=PDOM_Legacy, shadingModel=MSM_Unlit, usedShadingModels=Unlit}
 nodes [
  n1 = DynamicParameter() {paramNames [DistortionAmount, Brightness, UVDistortTile, UVDistort]}
  n2 = Multiply(A=TextureCoordinate() {uTiling=0.3, vTiling=0.3}, B=@n1.UVDistortTile)
  n3 = ComponentMask(None=TextureSample(UVs=Panner(Coordinate=@n2) {speedX=0.25, speedY=0.5}) {samplerType=SAMPLERTYPE_Normal, texture=/DealershipSharedLibrary/FX/Textures/Normal/T_VCVFX_FlowN03_N}.RGBA) {g=true, r=true}
  n4 = Panner(Coordinate=Add(A=TextureCoordinate(), B=Multiply(A=@n3, B=@n1.UVDistort)), Time=Constant() {r=1}) {speedX=-0.04}
  n5 = TextureSample(UVs=@n4) {samplerType=SAMPLERTYPE_Normal, texture=/DealershipSharedLibrary/FX/Textures/Normal/T_VCVFX_SmokeMask02_N}
  n6 = Constant() {r=0.35}
  n7 = Saturate(None=Add(
    A=TextureSampleParameter2D(UVs=@n4) {parameterName=Texture, texture=/DealershipSharedLibrary/FX/Textures/Gradient_Masks/T_VCVFX_Flame_A}.RGBA
    B=Multiply(A=@n5.A, B=@n6)
   ))
  n8 = ParticleColor()
  MaterialFunctionCall(Emissive=Multiply(A=Multiply(A=@n7, B=@n8.RGBA), B=@n1.Brightness)) {materialFunction=&MF_FX_EmissiveAdjustments}
  n9 = TextureSample() {samplerType=SAMPLERTYPE_Normal, texture=&T_VCVFX_SmokeMask02_N}
  n10 = Saturate(None=Multiply(A=@n8.A, B=Multiply(A=@n7, B=@n9.A)))
  Multiply(
   A=@n10
   B=MaterialFunctionCall(Opacity In=@n10, DepthFade Distance=Constant() {r=50}) {materialFunction=&MF_VCVFX_DepthFade}.Result
  )
  Multiply(A=@n5.A, B=@n1.DistortionAmount)
 ]
}
