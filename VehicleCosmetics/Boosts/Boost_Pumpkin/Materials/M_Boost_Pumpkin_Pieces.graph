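# Masked Default Lit Niagara sprite material for pumpkin pieces in the Pumpkin vehicle boost cosmetic effect.
# A TextureSwitch dynamic parameter selects between two piece variants, blending normals and alpha and hard-switching diffuse.
# Brightness-scaled diffuse plus a normal-driven Fresnel rim, tinted by squared GlowColor and scaled by FresnelBrightness, feeds MF_FX_EmissiveAdjustments.
# Particle alpha multiplies the texture alpha; the AnimationSpeed parameter is declared but unused.
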Material /VehicleCosmetics/Boosts/Boost_Pumpkin/Materials/M_Boost_Pumpkin_Pieces
graph {
 properties {bUsedWithNiagaraSprites=true, blendMode=BLEND_Masked, pixelDepthOffsetMode=PDOM_Legacy, usedShadingModels=Default Lit}
 nodes [
  n1 = TextureCoordinate()
  n2 = TextureSampleParameter2D(UVs=@n1) {
   parameterName=Normal_01
   samplerType=SAMPLERTYPE_LinearColor
   texture=/VehicleCosmetics/Boosts/Boost_Pumpkin/Textures/T_VCVFX_Boost_Pumpkin_Piece_01_FB_N
  }
  n3 = TextureSampleParameter2D(UVs=@n1) {
   parameterName=Normal_02
   samplerType=SAMPLERTYPE_LinearColor
   texture=/VehicleCosmetics/Boosts/Boost_Pumpkin/Textures/T_VCVFX_Boost_Pumpkin_Piece_02_FB_N
  }
  n4 = DynamicParameter() {paramNames [AnimationSpeed, Brightness, FresnelBrightness, TextureSwitch]}
  n5 = LinearInterpolate(A=@n2.RGB, B=@n3.RGB, Alpha=@n4.TextureSwitch)
  Fresnel(Normal=@n5) {exponent=3}
  n6 = TextureSampleParameter2D(UVs=@n1) {parameterName=Diffuse_02, texture=/VehicleCosmetics/Boosts/Boost_Pumpkin/Textures/T_VCVFX_Boost_Pumpkin_Piece_02_FB_D}
  n7 = Saturate(None=If(
    A=@n4.TextureSwitch
    B=Constant() {r=0.49}
    "A > B"=@n6.RGB
    "A == B"=@n6.RGB
    "A < B"=TextureSampleParameter2D(UVs=@n1) {parameterName=Diffuse_01, texture=/VehicleCosmetics/Boosts/Boost_Pumpkin/Textures/T_VCVFX_Boost_Pumpkin_Piece_01_FB_D}.RGB
   ))
  n8 = VectorParameter() {parameterName=GlowColor, defaultValue=<rgba 0.891 0.268 0.0385 8>}
  n9 = Multiply(A=@n8.RGB, B=@n8.RGB)
  n10 = MaterialFunctionCall(
   Normal Vector=@n5
   Camera Vector=Transform(None=CameraVectorWS()) {transformSourceType=TRANSFORMSOURCE_World, transformType=TRANSFORM_Tangent}
   Power=Constant() {r=3}
  ) {materialFunction=&Fresnel_Function}
  n11 = Multiply(A=@n4.FresnelBrightness, B=Saturate(None=Multiply(A=@n10.Result) {constB=2}))
  MaterialFunctionCall(Emissive=Add(A=Multiply(A=@n7, B=@n4.Brightness), B=Multiply(A=@n9, B=@n11))) {materialFunction=&MF_FX_EmissiveAdjustments}
  Reroute(None=@n5)
  Saturate(None=Multiply(A=ParticleColor().A, B=LinearInterpolate(A=@n2.A, B=@n3.A, Alpha=@n4.TextureSwitch)))
 ]
}
