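# Translucent unlit Niagara material for the trail of the Pumpkin vehicle boost.
# A panning flow normal map distorts second-UV lookups into liquid trail and smoke textures, which combine with a procedural center band.
# UV gradients fade that mask at both trail ends, and particle alpha scales it into a saturated opacity-style value.
# A similar UV mask blends a constant color toward particle color for the shared emissive adjustment function.
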Material /VehicleCosmetics/Boosts/Boost_Pumpkin/Materials/M_Boost_TrailPumpkin_V2
graph {
 properties {
  bAutomaticallySetUsageInEditor=false
  bUsedWithNiagaraMeshParticles=true
  bUsedWithNiagaraRibbons=true
  bUsedWithNiagaraSprites=true
  blendMode=BLEND_Translucent
  shadingModel=MSM_Unlit
  usedShadingModels=Unlit
 }
 nodes [
  n1 = Constant3Vector()
  n2 = ParticleColor()
  n3 = TextureCoordinate()
  n4 = ComponentMask(None=@n3) {g=true}
  n5 = Multiply(A=@n4, B=OneMinus(None=@n4))
  n6 = Saturate(None=ConstantBiasScale(None=@n5) {bias=-0.175, scale=5})
  n7 = ComponentMask(None=@n3) {r=true}
  n8 = Saturate(None=OneMinus(None=@n7))
  MaterialFunctionCall(Emissive=LinearInterpolate(A=Multiply(A=@n1), B=@n2.RGB, Alpha=Multiply(A=@n8, B=@n6))) {materialFunction=&MF_FX_EmissiveAdjustments}
  n9 = TextureSample(UVs=Panner(Coordinate=TextureCoordinate() {uTiling=0.5, vTiling=0.5}) {speedX=-0.5, speedY=0.25}) {samplerType=SAMPLERTYPE_Normal, texture=/DealershipSharedLibrary/FX/Textures/Normal/T_VCVFX_FlowN03_N}
  n10 = Add(
   A=Panner(Coordinate=TextureCoordinate() {coordinateIndex=1}, Time=Constant() {r=0.5}) {speedY=0.07}
   B=Multiply(A=ComponentMask(None=@n9.RGB) {g=true, r=true}, B=Constant() {r=0.5})
  )
  n11 = TextureSample(UVs=@n10) {samplerType=SAMPLERTYPE_LinearColor, texture=/VehicleCosmetics/Boosts/Boost_Pumpkin/Textures/T_VCVFX_Boost_Pumpkin_LiquidTrail_N}
  Reroute(None=@n11.RGB)
  n12 = Multiply(A=Constant() {r=25}, B=Power(Base=@n5, Exp=Constant() {r=4}))
  n13 = TextureSample(UVs=@n10) {texture=/VehicleCosmetics/Boosts/Boost_Pumpkin/Textures/T_VCVFX_Boost_Pumpkin_Smoke_Trail_01_RGB}
  n14 = Multiply(A=@n11.A, B=Add(A=Desaturation(None=@n11.RGB) {luminanceFactors=<rgba 0.3 0.59 0.11 1>}, B=Multiply(A=@n13.B)))
  n15 = Constant() {r=2}
  n16 = Multiply(A=@n8, B=Add(A=Saturate(None=@n12), B=Multiply(A=@n14, B=@n15)))
  n17 = Multiply(A=@n7, B=Constant() {r=8})
  Saturate(None=Multiply(A=@n2.A, B=Multiply(A=@n16, B=@n17)))
 ]
}
