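# Translucent unlit Niagara material for the vine trail of the Pumpkin vehicle boost.
# Two vine texture layers scroll with shared noise distortion, with dynamic parameters driving tiling, velocity and per-vine offsets.
# Alpha-masked vine color is tinted by glow or particle color for emissive, while opacity fades along U and follows particle alpha.
# A static-switched TrailBend builds a time-animated sine/cosine wave offset that ramps in from the U start.
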Material /VehicleCosmetics/Boosts/Boost_Pumpkin/Materials/M_Boost_VinesTrail_Opt
graph {
 properties {
  bAutomaticallySetUsageInEditor=false
  bUsedWithNiagaraMeshParticles=true
  bUsedWithNiagaraRibbons=true
  bUsedWithNiagaraSprites=true
  blendMode=BLEND_Translucent
  shadingModel=MSM_Unlit
  twoSided=true
  usedShadingModels=Unlit
 }
 nodes [
  Constant() {r=1}
  n1 = ParticleColor()
  n2 = StaticSwitchParameter(True=@n1.RGB, False=ScalarParameter() {defaultValue=1, parameterName=VineGlow}) {parameterName=ColorOverride}
  n3 = DynamicParameter() {defaultValue=<rgba 0 0.16 -1 0>, paramNames [Tilling, Offset_Vine_1, Velocity, Offset_Vine_2]}
  n4 = Multiply(A=TextureCoordinate() {vTiling=0}, B=@n3.Tilling) {constA=1}
  n5 = AppendVector(A=@n3.Offset_Vine_1, B=Constant())
  n6 = Multiply(A=@n3.Velocity, B=Constant2Vector() {r=1})
  n7 = TextureSample(UVs=Panner(Coordinate=TextureCoordinate() {uTiling=0.5, vTiling=0.25}) {speedX=-0.25, speedY=0.25}) {texture=/DealershipSharedLibrary/FX/Textures/Noise/T_VCVFX_Flow02_N}
  n8 = Add(
   A=LinearInterpolate(A=TextureCoordinate() {vTiling=-1}, B=TextureCoordinate() {uTiling=-1}) {constAlpha=0}
   B=Multiply(A=ComponentMask(None=@n7.RGB) {g=true, r=true}, B=Constant() {r=0.05})
  )
  n9 = Add(A=Panner(Coordinate=Add(A=@n4, B=@n5), Time=Time(), Speed=@n6) {speedX=-2}, B=@n8)
  n10 = TextureSample(UVs=@n9) {samplerType=SAMPLERTYPE_LinearColor, texture=/VehicleCosmetics/Boosts/Boost_Pumpkin/Textures/T_VCVFX_Boost_Pumpkin_Vine_N}
  n11 = Multiply(A=@n3.Tilling, B=TextureCoordinate() {vTiling=0}) {constA=1}
  n12 = AppendVector(A=@n3.Offset_Vine_2, B=Constant())
  n13 = Add(A=Panner(Coordinate=Add(A=@n11, B=@n12), Time=Time(), Speed=@n6) {speedX=-2}, B=@n8) {constA=1}
  n14 = TextureSample(UVs=@n13) {samplerType=SAMPLERTYPE_LinearColor, texture=&T_VCVFX_Boost_Pumpkin_Vine_N}
  n15 = Add(
   A=Multiply(A=TextureSample(UVs=@n9) {texture=/DealershipSharedLibrary/FX/Textures/Diffuse/T_VCVFX_Vine_D}.RGB, B=@n10.A)
   B=Multiply(A=TextureSample(UVs=@n13) {texture=&T_VCVFX_Vine_D}.RGB, B=@n14.A)
  )
  MaterialFunctionCall(Emissive=Saturate(None=Multiply(A=@n2, B=@n15))) {materialFunction=&MF_FX_EmissiveAdjustments}
  n16 = ComponentMask(None=TextureCoordinate()) {r=true}
  n17 = Multiply(A=Constant() {r=8}, B=@n16)
  n18 = Saturate(None=Multiply(A=Saturate(None=@n17), B=OneMinus(None=@n16)))
  n19 = Multiply(A=@n18, B=@n10.A)
  n20 = Multiply(A=@n18, B=@n14.A)
  Saturate(None=Multiply(A=Add(A=@n19, B=@n20), B=@n1.A))
  ScalarParameter() {defaultValue=1, parameterName=Streght}
  n21 = VectorParameter() {parameterName=WaveStreght, defaultValue=<rgba 1 1 1 0>}
  n22 = Multiply(A=@n21.G, B=OneMinus(None=Time()))
  n23 = ComponentMask(None=Multiply(A=TextureCoordinate() {uTiling=2}, B=@n21.R)) {r=true}
  n24 = Add(A=@n22, B=@n23)
  n25 = Add(A=Cosine(None=@n24), B=Cosine(None=@n24) {period=2})
  n26 = Add(A=@n23, B=@n22)
  n27 = Add(A=Sine(None=@n26), B=Sine(None=@n26) {period=2})
  n28 = Transform(None=Multiply(B=MaterialFunctionCall(X=@n25, Y=@n27, Z=@n25) {materialFunction=&MakeFloat3}.Result) {constA=1})
  n29 = Constant() {r=8}
  n30 = SmoothStep(Value=ComponentMask(None=TextureCoordinate()) {r=true}) {constMax=0.1}
  StaticSwitchParameter(True=Multiply(A=Multiply(A=@n28, B=@n21.B), B=Multiply(A=@n29, B=@n30)), False=Constant()) {defaultValue=true, parameterName=TrailBend}
 ]
}
